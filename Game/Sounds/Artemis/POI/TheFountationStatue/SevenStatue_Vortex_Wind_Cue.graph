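# Looping wind ambience cue for the vortex at the Seven statue point of interest.
# A single wave player loops a reused cooling tower wind recording through an outdoor ambient ducking sound class.
# Overridden attenuation uses a large capsule with logarithmic falloff and wide stereo spread, plus interpolated listener focus that lowers off-focus volume and occlusion that muffles and quiets blocked sound.
# The debug flag is left enabled.
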SoundCue /Game/Sounds/Artemis/POI/TheFountationStatue/SevenStatue_Vortex_Wind_Cue
root=WavePlayer_0
nodes [{id=WavePlayer_0, properties {bLooping=true, soundWaveAssetPtr=/Game/Sounds/Fort_Prop_Sounds/Exterior/CoolingTower/CoolingTowerWind_01}}]
properties {
 bDebug=true
 bOverrideAttenuation=true
 maxDistance=10300
 soundClassObject=/Game/Sounds/FortSoundClassesAndModes/Fort_Ambient_DuckOutdoors_Athena_PSM
 volumeMultiplier=0.902
 attenuationOverrides {
  attenuationShape=Capsule
  bEnableFocusInterpolation=true
  bEnableListenerFocus=true
  bEnableOcclusion=true
  distanceAlgorithm=Logarithmic
  falloffDistance=2800
  nonFocusVolumeAttenuation=0.7
  nonSpatializedRadiusStart=1500
  occlusionLowPassFilterFrequency=10000
  occlusionTraceChannel=ECC_GameTraceChannel11
  occlusionVolumeAttenuation=0.5
  stereoSpread=10000
  attenuationShapeExtents=<7500 1500>
 }
}
